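# M_NestTracker is a translucent UI material that draws a nest progress tracker as repeated square cells.
# Scaled UI UVs are offset and wrapped into cells, with NestsCompleted and NestAmount stepped against cell position to mask states.
# Completed and incomplete color composites are merged by those masks, alongside a square opacity composite.
# Cells combine SDF rounded boxes and outlines, checkmark and shadow masks from one texture, a curve-atlas gradient and a loot-sac image.
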Material /TidalCraneSharedActivitiesUI/Materials/M_NestTracker
graph {
 properties {bAutomaticallySetUsageInEditor=false, blendMode=BLEND_Translucent, materialDomain=MD_UI, usedShadingModels=Default Lit}
 nodes [
  n1 = NamedRerouteUsage()
  n2 = Constant() {r=4}
  n3 = MaterialFunctionCall(Float2=MaterialFunctionCall(
    UVs=MaterialFunctionCall!(UVs=@n1, Sides=@n2) {materialFunction=/FortUILibrary/MaterialFunctions/Basics/MF_UI_UVRadialMirror}.UVs
    "Rotation Angle (0-1)"=Constant() {r=0.25}
   ) {materialFunction=&CustomRotator}.Rotated Values) {materialFunction=&BreakOutFloat2Components}
  n4 = Add(A=@n3.R) {constB=0.4}
  MaterialFunctionCall(DistanceField=MaterialFunctionCall!(UVs=AppendVector(A=@n4, B=@n3.G), Size=Constant2Vector() {g=0.125, r=0.8}, CornerRoundness, KeepAspectRatio, PixelSize) {materialFunction=/FortUILibrary/MaterialFunctions/Shapes/SDFs/MF_SDF_Box}.SDF) {materialFunction=&MF_UI_AntiAliasing}
  n5 = MaterialFunctionCall() {materialFunction=&GetUserInterfaceUV}
  n6 = MaterialFunctionCall(Float2=MaterialFunctionCall(UV=@n5.Normalized UV) {materialFunction=&MF_UI_UVScaling}.ScaleY) {materialFunction=&BreakOutFloat2Components}
  n7 = MaterialFunctionCall(Float2=@n5.Pixel Size) {materialFunction=&BreakOutFloat2Components}
  n8 = Floor(None=Divide(A=@n7.R, B=@n7.R) {desc=Aspect Ratio}) {desc=Total Cells}
  n9 = Multiply(A=@n8) {constB=-0.5, desc=OffsetFirstCell}
  n10 = Clamp(None=Add(A=Subtract(A=@n6.R, B=@n9)) {constB=-0.5, desc=OffsetCellFromSide}, Max=@n8) {desc=StretchEdge, maxDefault=10}
  NamedRerouteDeclaration(None=Step(Y=@n10, X=ScalarParameter() {defaultValue=1, parameterName=NestsCompleted})) {name=Input_NestsCompleted, nodeColor=<rgba 0.75 0.6098 0.375 0.75>}
  n11 = LinearInterpolate(
   A=LinearInterpolate(B=NamedRerouteUsage(), Alpha=NamedRerouteUsage())
   B=Constant3Vector() {constant=<rgba 0.0578 0.5647 0.2705 1>}
   Alpha=NamedRerouteUsage()
  )
  n12 = NamedRerouteUsage()
  n13 = NamedRerouteDeclaration(None=LinearInterpolate(A=LinearInterpolate(A=@n11, Alpha=@n12) {constB=0}, B=Constant3Vector() {constant=<rgba 0 0.6867 0.4233 1>}, Alpha=NamedRerouteUsage())) {name=Composite_CompleteColors, nodeColor=<rgba 0.75 0.6098 0.375 0.75>}
  n14 = NamedRerouteUsage()
  n15 = LinearInterpolate(B=Constant(), Alpha=NamedRerouteUsage())
  n16 = Constant3Vector() {constant=<rgba 0.1683 0.1144 0.0844 1>}
  n17 = NamedRerouteUsage()
  n18 = NamedRerouteDeclaration(None=LinearInterpolate(A=LinearInterpolate(A=@n15, B=@n16, Alpha=@n17), B=Constant() {r=0.2}, Alpha=NamedRerouteUsage()) {constB=0}) {name=Composite_IncompleteColors, nodeColor=<rgba 0.75 0.6098 0.375 0.75>}
  n19 = Subtract(A=NamedRerouteUsage(), B=NamedRerouteUsage())
  NamedRerouteDeclaration(None=Max(A=Multiply(A=@n13, B=@n14), B=Multiply(A=@n18, B=@n19))) {name=Composite_NestStatusStates, nodeColor=<rgba 0.75 0.6098 0.375 0.75>}
  n20 = MaterialFunctionCall(UVs=NamedRerouteUsage(), Size=Constant() {r=0.6}, CornerRoundness=Constant() {r=0.05}) {materialFunction=&MF_SDF_Box}
  n21 = NamedRerouteDeclaration(None=MaterialFunctionCall!(DistanceField=@n20.SDF, Threshold, Thickness=Constant() {r=0.035}, Alignment=Constant() {r=3}, Smoothing) {materialFunction=/FortUILibrary/MaterialFunctions/Shapes/Modifiers/MF_UI_ShapeOutline}.Outline) {name=Mask_BoxOutline, nodeColor=<rgba 0.75 0.6098 0.375 0.75>}
  n22 = NamedRerouteDeclaration(None=MaterialFunctionCall(DistanceField=@n20.SDF) {materialFunction=&MF_UI_AntiAliasing}.Inner) {name=Mask_BoxShape, nodeColor=<rgba 0.75 0.6098 0.375 0.75>}
  n23 = Max(A=@n21, B=Multiply(A=@n22))
  n24 = Subtract(A=NamedRerouteUsage(), B=NamedRerouteUsage())
  n25 = Max(A=@n21, B=@n22)
  n26 = NamedRerouteUsage()
  NamedRerouteDeclaration(None=Max(A=Multiply(A=@n23, B=@n24), B=Multiply(A=@n25, B=@n26))) {name=Composite_SquareOpacity, nodeColor=<rgba 0.75 0.6098 0.375 0.75>}
  n27 = AppendVector(A=@n10, B=@n6.G)
  NamedRerouteDeclaration(None=MaterialFunctionCall(UVs=Frac(None=@n27), Texture Scale=Constant() {r=1.2}) {materialFunction=&ScaleUVsByCenter}.UVs) {name=RepeatUVs, nodeColor=<rgba 0.75 0.6098 0.375 0.75>}
  n28 = TextureSample(UVs=MaterialFunctionCall(UVs=@n1, Texture Scale=Constant() {r=0.7}) {materialFunction=&ScaleUVsByCenter}.UVs) {texture=/TidalCraneSharedActivitiesUI/Textures/T_LootSacRender}
  NamedRerouteDeclaration(None=LinearInterpolate(A=Constant3Vector() {constant=<rgba 1 0.4207 0.1117 1>}, B=@n28.RGB, Alpha=@n28.RGB)) {name=Mask_XMark, nodeColor=<rgba 0.75 0.6098 0.375 0.75>}
  n29 = TextureSample(UVs=MaterialFunctionCall(UVs=NamedRerouteUsage(), Texture Scale=Constant() {r=0.5}) {materialFunction=&ScaleUVsByCenter}.UVs) {samplerType=SAMPLERTYPE_LinearGrayscale, texture=/TidalCraneSharedActivitiesUI/Textures/T_Hive_Checkmark}
  NamedRerouteDeclaration(None=MaterialFunctionCall(DistanceField=@n29.RGB, Threshold=Constant() {r=0.75}) {materialFunction=&MF_UI_AntiAliasing}.Outer) {name=Mask_CheckMark, nodeColor=<rgba 0.75 0.6098 0.375 0.75>}
  n30 = NamedRerouteUsage()
  n31 = Constant() {r=0.1}
  n32 = CurveAtlasRowParameter(CurveTime=MaterialFunctionCall(Float2=MaterialFunctionCall(UVs=@n30, "Rotation Angle (0-1)"=@n31) {materialFunction=&CustomRotator}.Rotated Values) {materialFunction=&BreakOutFloat2Components}.R) {
   atlas=/TidalCraneSharedActivitiesUI/Curves/CAt_HivePOI
   curve=/TidalCraneSharedActivitiesUI/Curves/C_CheckmarkGradient
   defaultValue=2
   parameterName=CheckmarkCurves
  }
  NamedRerouteDeclaration(None=@n32) {name=Color_CompleteBox, nodeColor=<rgba 0.75 0.6098 0.375 0.75>}
  NamedRerouteDeclaration(None=MaterialFunctionCall(DistanceField=@n29.RGB, Threshold=Constant() {r=0.2}) {materialFunction=&MF_UI_AntiAliasing}.Outer) {name=Mask_CheckMarkShadow, nodeColor=<rgba 0.75 0.6098 0.375 0.75>}
  NamedRerouteDeclaration(None=Step(Y=@n10, X=ScalarParameter() {defaultValue=6, parameterName=NestAmount}) {constX=4}) {name=Input_NestAmount, nodeColor=<rgba 0.75 0.6098 0.375 0.75>}
  NamedRerouteUsage()
 ]
}
