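# Translucent UI material for a hive progress bar that, per its name, shows progress toward summoning the queen.
# SDF shapes with mirrored and 9-slice UVs build anti-aliased accent, dot, frame and bar masks, shared through named reroutes.
# Progress clips the padded bar fill, while a curve-atlas gradient and InactiveOrActive-driven UVs shape its coloring.
# Lerp and Max chains composite these masks into color and opacity.
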Material /TidalCraneSharedActivitiesUI/Materials/M_ProgressBarHive_ProgressTowardQueenSummon
graph {
 properties {bAutomaticallySetUsageInEditor=false, blendMode=BLEND_Translucent, materialDomain=MD_UI, usedShadingModels=Default Lit}
 nodes [
  n1 = Constant3Vector() {constant=<rgba 0.2292 1 0.2409 1>}
  n2 = LinearInterpolate(A=@n1, B=@n1, Alpha=NamedRerouteUsage())
  n3 = NamedRerouteUsage()
  LinearInterpolate(
   A=LinearInterpolate(A=LinearInterpolate(B=@n2, Alpha=@n3), B=NamedRerouteUsage(), Alpha=NamedRerouteUsage())
   B=Constant() {r=1}
   Alpha=Max(A=NamedRerouteUsage(), B=NamedRerouteUsage())
  )
  n4 = Max(A=NamedRerouteUsage(), B=NamedRerouteUsage())
  Max(
   A=Max(A=NamedRerouteUsage(), B=Multiply(A=@n4) {constB=0.5})
   B=LinearInterpolate(A=NamedRerouteUsage(), B=NamedRerouteUsage()) {constAlpha=0.4}
  )
  n5 = MaterialFunctionCall() {desc=Uvs, materialFunction=&GetUserInterfaceUV}
  n6 = MaterialFunctionCall(Float2=@n5.Normalized UV) {materialFunction=&BreakOutFloat2Components}
  n7 = LinearInterpolate(A=@n5.Normalized UV, B=AppendVector(A=OneMinus(None=@n6.R), B=@n6.G), Alpha=Round(None=@n6.R))
  n8 = Constant2Vector() {g=1}
  n9 = MaterialFunctionCall(
   UVs=Add(
    A=MaterialFunctionCall(UV=@n7, ScalePivotPoint=@n8, ContainerSize=@n5.Pixel Size) {materialFunction=&MF_UI_UVScaling}.Scaled to Fit
    B=Constant2Vector() {r=0.35}
   ) {desc=Offset}
   "Rotation Angle (0-1)"=Constant()
  ) {materialFunction=&CustomRotator}
  n10 = MaterialFunctionCall(UVs=@n9.Rotated Values, "Rotation Angle (0-1)"=Constant() {r=0.25}) {materialFunction=&CustomRotator}
  n11 = Constant2Vector() {g=-0.1}
  n12 = Constant() {r=0.6}
  n13 = Multiply(A=MaterialFunctionCall!(UVs=Add(A=@n10.Rotated Values, B=@n11) {constB=0.02}, TopWidth=Subtract(A=@n12) {constB=0.2}, BottomWidth=@n12, Height=Constant() {r=0.2}) {materialFunction=/FortUILibrary/MaterialFunctions/Shapes/SDFs/MF_SDF_IsoscelesTrapezoid}.SDF) {constB=-1}
  n14 = MaterialFunctionCall!(UVs=@n9.Rotated Values, Size=Constant2Vector() {g=0.7, r=0.1}, CornerRoundness, KeepAspectRatio, PixelSize) {materialFunction=/FortUILibrary/MaterialFunctions/Shapes/SDFs/MF_SDF_Box}
  NamedRerouteDeclaration(None=MaterialFunctionCall(DistanceField=Max(A=@n13, B=@n14.SDF)) {desc=OuterFlareMask, materialFunction=&MF_UI_AntiAliasing}.Inner) {name=Mask_OuterAccents, nodeColor=<rgba 0.75 0.6098 0.375 0.75>}
  n15 = Add(
   A=MaterialFunctionCall(UV=@n7, ScalePivotPoint=@n8) {materialFunction=&MF_UI_UVScaling}.Scaled to Fit
   B=Constant2Vector() {g=0.3}
  ) {desc=Offset}
  n16 = Constant() {r=0.075}
  NamedRerouteDeclaration(None=MaterialFunctionCall(DistanceField=MaterialFunctionCall(UVs=@n15, Size=@n16) {materialFunction=&MF_SDF_Box}.SDF) {desc=DoubleDotMask, materialFunction=&MF_UI_AntiAliasing}.Inner) {name=Mask_DoubleDots, nodeColor=<rgba 0.75 0.6098 0.375 0.75>}
  n17 = MaterialFunctionCall!(
   DistanceField=MaterialFunctionCall(
    UVs=MaterialFunctionCall!(UVs=@n5.Normalized UV) {materialFunction=/FortUILibrary/MaterialFunctions/Basics/MF_UI_9SliceUVs}.UVs
    Size=Constant2Vector() {g=0.6, r=0.4}
   ) {materialFunction=&MF_SDF_Box}.SDF
   Threshold
   Thickness=Constant() {r=0.05}
   Alignment=Constant()
   Smoothing
  ) {materialFunction=/FortUILibrary/MaterialFunctions/Shapes/Modifiers/MF_UI_ShapeOutline}
  n18 = MaterialFunctionCall(Float2=@n15) {materialFunction=&BreakOutFloat2Components}
  n19 = Min(A=Subtract(A=@n18.R) {constB=0.425}, B=OneMinus(None=Add(A=@n18.G) {constB=0.45}))
  NamedRerouteDeclaration(None=MaterialFunctionCall(DistanceField=Max(A=@n19, B=@n17.SDFOutline)) {desc=OuterFrameMask, materialFunction=&MF_UI_AntiAliasing}.Inner) {name=Mask_OuterFrame, nodeColor=<rgba 0.75 0.6098 0.375 0.75>}
  n20 = ScalarParameter() {defaultValue=0.8, parameterName=Padding}
  n21 = MaterialFunctionCall() {desc=Uvs, materialFunction=&GetUserInterfaceUV}
  n22 = MaterialFunctionCall(Float2=@n21.Pixel Size) {materialFunction=&BreakOutFloat2Components}
  n23 = Divide(A=@n22.G, B=@n22.G)
  n24 = OneMinus(None=AppendVector(A=Multiply(A=@n20, B=@n23), B=@n20))
  n25 = MaterialFunctionCall(DistanceField=MaterialFunctionCall(Size=@n24) {materialFunction=&MF_SDF_Box}.SDF) {desc=BarBGMask, materialFunction=&MF_UI_AntiAliasing}
  n26 = ComponentMask(None=MaterialFunctionCall(UVs=@n21.Normalized UV, Texture Scale=@n24) {materialFunction=&ScaleUVsByCenter}.UVs) {r=true}
  n27 = ScalarParameter() {defaultValue=0.5, parameterName=Progress}
  NamedRerouteDeclaration(None=Multiply(A=@n25.Inner, B=Step(Y=@n26, X=@n27) {constX=0.9, constY=1}) {desc=ProgressMask}) {name=Mask_ProgressPercent, nodeColor=<rgba 0.75 0.6098 0.375 0.75>}
  NamedRerouteDeclaration(None=@n25.Inner) {name=Mask_BarBG, nodeColor=<rgba 0.75 0.6098 0.375 0.75>}
  n28 = MaterialFunctionCall(Float2=LinearInterpolate(A=Constant2Vector() {r=1}, B=@n5.Normalized UV, Alpha=ScalarParameter() {defaultValue=1, parameterName=InactiveOrActive}) {constA=1}) {materialFunction=&BreakOutFloat2Components}
  n29 = Add(A=@n28.G) {constB=-0.545}
  NamedRerouteDeclaration(None=Multiply(A=@n25.Inner, B=Ceil(None=@n29)) {desc=BottomColorMask}) {name=Mask_BottomProgressColor, nodeColor=<rgba 0.75 0.6098 0.375 0.75>}
  n30 = OneMinus(None=@n28.G)
  n31 = CurveAtlasRowParameter(CurveTime=Multiply(A=Add(A=@n30) {constB=-0.45}) {constB=8}) {
   atlas=/TidalCraneSharedActivitiesUI/Curves/CAt_HivePOI
   curve=/TidalCraneSharedActivitiesUI/Curves/C_SacGradient
   desc=Color Mask
   parameterName=ProgressColors
  }
  NamedRerouteDeclaration(None=@n31) {name=Color_ProgressGradient, nodeColor=<rgba 0.75 0.6098 0.375 0.75>}
 ]
}
