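# UI material for the Reclamation progress bar in Tidal Crane shared activities, with Bug Boss and Bomber Bug kill variations.
# Antialiased SDF masks build the frame, accents, dots, background, progress and delta fills, shared through named reroutes into color and alpha composite.
# Scalar parameters drive progress, flashes, reveal and boss recoloring, alongside quantized-time scratch noise and panning boss textures.
# Static switches select notched progress or a gradient-only output.
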Material /TidalCraneSharedActivitiesUI/Materials/M_UI_Reclamation_Progress
graph {
 properties {bAutomaticallySetUsageInEditor=false, blendMode=BLEND_AlphaComposite, materialDomain=MD_UI, usedShadingModels=Default Lit}
 nodes [
  Constant2Vector() {g=0.44}
  Constant2Vector() {r=0.35}
  Constant()
  n1 = MaterialFunctionCall() {desc=Uvs, materialFunction=&GetUserInterfaceUV}
  n2 = MaterialFunctionCall(Float2=@n1.Normalized UV) {materialFunction=&BreakOutFloat2Components}
  n3 = LinearInterpolate(A=@n1.Normalized UV, B=AppendVector(A=OneMinus(None=@n2.R), B=@n2.G), Alpha=Round(None=@n2.R))
  n4 = Constant2Vector() {g=1}
  n5 = LinearInterpolate(Alpha=ScalarParameter() {parameterName=Outer Accent X Movement}) {constA=0.35, constB=-4}
  n6 = Constant()
  n7 = MaterialFunctionCall(
   UVs=Add(
    A=MaterialFunctionCall(UV=@n3, ScalePivotPoint=@n4, ContainerSize=@n1.Pixel Size) {materialFunction=&MF_UI_UVScaling}.Scaled to Fit
    B=AppendVector(A=@n5, B=@n6)
   ) {desc=Offset}
   "Rotation Angle (0-1)"=Constant()
  ) {materialFunction=&CustomRotator}
  n8 = Add(
   A=MaterialFunctionCall(UVs=@n7.Rotated Values, "Rotation Angle (0-1)"=Constant() {r=0.25}) {materialFunction=&CustomRotator}.Rotated Values
   B=Constant2Vector() {g=-0.1}
  ) {constB=0.02}
  n9 = Constant() {r=0.6}
  n10 = Subtract(A=@n9) {constB=0.2}
  n11 = Constant() {r=0.2}
  n12 = Max(
   A=Multiply(A=MaterialFunctionCall!(UVs=@n8, TopWidth=@n10, BottomWidth=@n9, Height=@n11) {materialFunction=/FortUILibrary/MaterialFunctions/Shapes/SDFs/MF_SDF_IsoscelesTrapezoid}.SDF) {constB=-1}
   B=MaterialFunctionCall!(UVs=@n7.Rotated Values, Size=Constant2Vector() {g=0.7, r=0.1}, CornerRoundness, KeepAspectRatio, PixelSize) {materialFunction=/FortUILibrary/MaterialFunctions/Shapes/SDFs/MF_SDF_Box}.SDF
  )
  NamedRerouteDeclaration(None=Multiply(
    A=MaterialFunctionCall(DistanceField=@n12) {desc=OuterFlareMask, materialFunction=&MF_UI_AntiAliasing}.Inner
    B=ScalarParameter() {defaultValue=0.5, parameterName=Outer Accents Opacity}
   )) {name=Mask_OuterAccents, nodeColor=<rgba 0.75 0.6098 0.375 0.75>}
  n13 = Add(
   A=MaterialFunctionCall(UV=@n3, ScalePivotPoint=@n4) {materialFunction=&MF_UI_UVScaling}.Scaled to Fit
   B=Constant2Vector() {g=0.3, r=-0.95}
  ) {desc=Offset}
  n14 = MaterialFunctionCall(
   UVs=@n13
   Size=LinearInterpolate(A=Constant() {r=0.1}, Alpha=ScalarParameter() {group=Double Dots, parameterName=Double Dots Size}) {constA=0.13, constB=0.3}
   CornerRoundness=Constant() {r=0.02}
  ) {materialFunction=&MF_SDF_Box}
  NamedRerouteDeclaration(None=Multiply(
    A=MaterialFunctionCall(DistanceField=@n14.SDF) {desc=DoubleDotMask, materialFunction=&MF_UI_AntiAliasing}.Inner
    B=ScalarParameter() {defaultValue=0.5, group=Double Dots, parameterName=Double Dots Opacity}
   )) {name=Mask_DoubleDots, nodeColor=<rgba 0.75 0.6098 0.375 0.75>}
  n15 = TextureCoordinate() {uTiling=3, vTiling=3}
  n16 = Multiply(A=Frac(None=Time())) {constB=0.5}
  n17 = Constant() {r=12}
  n18 = StaticBool()
  n19 = TextureSample(UVs=Add(
    A=MaterialFunctionCall(UV=@n15) {materialFunction=&MF_UI_UVScaling}.Scaled To Fill
    B=MaterialFunctionCall!("Param (0-1)"=@n16, Tiles=@n17, Apply Offset=@n18) {materialFunction=/Game/Environments/Papaya/Materials/MaterialFunctions/MF_Papaya_Quantize_Scalar}.Result
   )) {samplerType=SAMPLERTYPE_LinearGrayscale, texture=/TidalCraneSharedActivitiesUI/Textures/T_UI_TidalCrane_Noise}
  n20 = TextureCoordinate()
  n21 = Multiply(A=Frac(None=Time())) {constB=2}
  n22 = Constant() {r=5}
  n23 = TextureSample(UVs=Add(
    A=MaterialFunctionCall(UV=@n20) {materialFunction=&MF_UI_UVScaling}.Scaled to Fit
    B=MaterialFunctionCall("Param (0-1)"=@n21, Tiles=@n22) {materialFunction=&MF_Papaya_Quantize_Scalar}.Result
   )) {samplerType=SAMPLERTYPE_Normal, texture=/Game/Global/Textures/Default/Noise/64x64Noise}
  NamedRerouteDeclaration(None=Add(A=Power(Base=@n19.R), B=Multiply(A=@n23.R) {constB=0.1})) {name=Scratches, nodeColor=<rgba 1 0.0469 0 1>}
  n24 = ScalarParameter() {defaultValue=0.8, parameterName=Padding}
  n25 = MaterialFunctionCall() {desc=Uvs, materialFunction=&GetUserInterfaceUV}
  n26 = MaterialFunctionCall(Float2=@n25.Pixel Size) {materialFunction=&BreakOutFloat2Components}
  n27 = Divide(A=@n26.G, B=@n26.G)
  n28 = OneMinus(None=AppendVector(A=Multiply(A=@n24, B=@n27), B=@n24))
  n29 = MaterialFunctionCall(Size=@n28) {materialFunction=&MF_SDF_Box}
  n30 = Multiply(A=NamedRerouteUsage()) {constB=0.05}
  n31 = MaterialFunctionCall(DistanceField=LinearInterpolate(A=@n29.SDF, B=Subtract(A=@n29.SDF, B=@n30), Alpha=ScalarParameter() {parameterName=BugBossProgressSwitch})) {desc=BarBGMask, materialFunction=&MF_UI_AntiAliasing}
  n32 = MaterialFunctionCall(UVs=@n25.Normalized UV, Texture Scale=@n28) {materialFunction=&ScaleUVsByCenter}
  n33 = ComponentMask(None=NamedRerouteDeclaration(None=Saturate(None=@n32.UVs)) {name=UVs, nodeColor=<rgba 0.1016 0 1 1>}) {r=true}
  n34 = SmoothStep(Value=@n33) {constMax=0.6}
  n35 = Step(Y=@n33, X=ScalarParameter() {desc=Default: 0, parameterName=Progress}) {constX=0.9, constY=1}
  NamedRerouteDeclaration(None=Multiply(A=@n31.Inner, B=Min(A=@n34, B=@n35))) {name=BomberBugFX_Alpha, nodeColor=<rgba 0 1 0.1094 1>}
  n36 = NamedRerouteUsage()
  n37 = Step(Y=ComponentMask(None=@n36) {r=true}, X=ScalarParameter() {desc=Default: 0, parameterName=Progress}) {constX=0.9, constY=1}
  n38 = MaterialFunctionCall(UV Channel=Add(A=TextureCoordinate(), B=Constant2Vector())) {materialFunction=&LinearGradient}
  n39 = Multiply(A=OneMinus(None=@n38.VGradient)) {constB=10}
  n40 = Multiply(
   A=MaterialFunctionCall(UVs=@n36, Size=Constant2Vector() {g=5, r=1}) {materialFunction=&MF_UI_Box}.Fill
   B=Multiply(A=Divide(B=@n39), B=Power(Base=@n38.VGradient) {constExponent=2.5})
  )
  n41 = Multiply(A=Saturate(None=Multiply(A=@n37, B=@n40))) {constB=0.7}
  n42 = Multiply(A=@n41, B=NamedRerouteUsage())
  NamedRerouteDeclaration(None=Multiply(A=Add(A=@n41, B=@n42)) {constB=0.1}) {name=Gradient, nodeColor=<rgba 0 1 0.9766 1>}
  n43 = Subtract(A=TextureCoordinate()) {constB=0.5}
  n44 = Abs(None=Multiply(A=ComponentMask(None=@n43) {r=true}) {constB=2})
  n45 = ScalarParameter() {defaultValue=1, parameterName=Progress Bar Reveal}
  NamedRerouteDeclaration(None=MaterialFunctionCall(DistanceField=@n44, Threshold=LinearInterpolate(Alpha=@n45) {constB=0.95}) {desc=OuterFrameMask, materialFunction=&MF_UI_AntiAliasing}.Inner) {name=Progress Bar Reveal Mask, nodeColor=<rgba 0.3594 1 0 1>}
  NamedRerouteDeclaration(None=MaterialFunctionCall(Speed=Constant() {r=0.1}) {materialFunction=&DebugTimeSine}.Result) {name=Progress Test, nodeColor=<rgba 0 0.0156 1 1>}
  n46 = Constant3Vector() {constant=<rgba 1 0.2085 0 1>}
  NamedRerouteDeclaration(None=@n46) {name=DotColor_Default, nodeColor=<rgba 0 0 0 1>}
  n47 = Constant3Vector() {constant=<rgba 0.6724 0.2582 0.6172 1>}
  NamedRerouteDeclaration(None=@n47) {name=DotColor_Boss, nodeColor=<rgba 0 0 0 1>}
  n48 = MaterialFunctionCall!(UVs=@n1.Normalized UV) {materialFunction=/FortUILibrary/MaterialFunctions/Basics/MF_UI_9SliceUVs}
  n49 = LinearInterpolate(A=Constant2Vector() {g=0.6, r=0.4}, B=Constant2Vector() {r=0.08}, Alpha=ScalarParameter() {parameterName=Border Collapse}) {constA=0.075, constAlpha=1, constB=0.035}
  n50 = ScalarParameter() {parameterName=Border Thickness}
  n51 = MaterialFunctionCall(Float2=@n13) {materialFunction=&BreakOutFloat2Components}
  n52 = Add(A=@n51.G) {constB=0.3}
  n53 = Max(
   A=Min(A=Subtract(A=@n51.R) {constB=0.4}, B=OneMinus(None=@n52))
   B=MaterialFunctionCall!(
    DistanceField=MaterialFunctionCall(UVs=@n48.UVs, Size=@n49) {materialFunction=&MF_SDF_Box}.SDF
    Threshold
    Thickness=LinearInterpolate(Alpha=@n50) {constA=0.075, constAlpha=0, constB=0.035}
    Alignment=Constant()
    Smoothing
   ) {materialFunction=/FortUILibrary/MaterialFunctions/Shapes/Modifiers/MF_UI_ShapeOutline}.SDFOutline
  )
  NamedRerouteDeclaration(None=Multiply(
    A=MaterialFunctionCall(DistanceField=@n53) {desc=OuterFrameMask, materialFunction=&MF_UI_AntiAliasing}.Inner
    B=ScalarParameter() {defaultValue=0.5, parameterName=Outer Frame Opacity}
   )) {name=Mask_OuterFrame, nodeColor=<rgba 0.75 0.6098 0.375 0.75>}
  n54 = NamedRerouteUsage()
  n55 = NamedRerouteUsage()
  n56 = Constant2Vector() {r=-0.0001}
  n57 = MaterialFunctionCall(DistanceField=ComponentMask(None=Add(A=@n55, B=@n56)) {r=true}) {desc=BarLeftEdgeCleanUp, materialFunction=&MF_UI_AntiAliasing}
  n58 = Constant2Vector() {r=0.0001}
  n59 = OneMinus(None=ComponentMask(None=Add(A=@n55, B=@n58)) {r=true})
  n60 = Multiply(
   A=@n54
   B=Multiply(
    A=Multiply(A=@n31.Inner, B=@n57.Outer)
    B=MaterialFunctionCall(DistanceField=@n59) {desc=BarLeftEdgeCleanUp, materialFunction=&MF_UI_AntiAliasing}.Outer
   )
  )
  n61 = Multiply(A=Multiply(A=@n60, B=@n35) {desc=ProgressMask}, B=ScalarParameter() {defaultValue=1, parameterName=Progress Opacity})
  n62 = NamedRerouteUsage()
  NamedRerouteDeclaration(None=StaticSwitchParameter(True=Multiply(A=@n61, B=@n62), False=@n61) {parameterName=Notched Progress}) {name=Mask_ProgressPercent, nodeColor=<rgba 0.75 0.6098 0.375 0.75>}
  n63 = Multiply(A=Multiply(
    A=@n54
    B=MaterialFunctionCall(DistanceField=@n29.SDF) {desc=BarBGMask, materialFunction=&MF_UI_AntiAliasing}.Inner
   )) {constB=0.5}
  NamedRerouteDeclaration(None=StaticSwitchParameter(True=Multiply(A=@n63, B=@n62), False=@n63) {parameterName=Notched Progress}) {name=Mask_BarBG, nodeColor=<rgba 0.75 0.6098 0.375 0.75>}
  n64 = TextureCoordinate() {uTiling=0.15, vTiling=2}
  n65 = Constant2Vector() {g=1.1931}
  n66 = Panner(
   Coordinate=MaterialFunctionCall(UV=Add(A=@n64, B=@n65)) {materialFunction=&MF_UI_UVScaling}.Scaled to Fit
   Speed=AppendVector(A=Constant() {r=0.2}, B=Constant())
  ) {speedX=-0.075, speedY=0.125}
  n67 = MaterialFunctionCall(UV=TextureCoordinate() {uTiling=0.15, vTiling=0.9}) {materialFunction=&MF_UI_UVScaling}
  n68 = AppendVector(A=Constant() {r=0.1}, B=Constant())
  n69 = OneMinus(None=TextureSample(UVs=Panner(Coordinate=@n67.Scaled to Fit, Speed=@n68) {speedX=-0.075, speedY=0.125}) {samplerType=SAMPLERTYPE_LinearGrayscale, texture=/TidalCraneSharedActivitiesUI/Textures/T_UI_BugBoss}.R)
  n70 = Multiply(A=Divide(
    A=TextureSample(UVs=@n66) {samplerType=SAMPLERTYPE_LinearGrayscale, texture=&T_UI_BugBoss}.R
    B=Multiply(A=@n69) {constB=3}
   )) {constB=4}
  NamedRerouteDeclaration(None=Multiply(A=Saturate(None=@n70)) {constB=0.7}) {name=Boss Queen Fx, nodeColor=<rgba 0 0 0 1>}
  n71 = ScalarParameter() {desc=Default: 0, parameterName=DeltaProgress}
  n72 = Multiply(
   A=Multiply(A=@n60, B=Step(Y=@n33, X=@n71) {constX=0.9, constY=1}) {desc=DeltaMask}
   B=ScalarParameter() {defaultValue=1, parameterName=Progress Opacity}
  )
  n73 = NamedRerouteUsage()
  NamedRerouteDeclaration(None=StaticSwitchParameter(True=Multiply(A=@n72, B=@n73), False=@n72) {parameterName=Notched Progress}) {name=DeltaMask_ProgressPercent, nodeColor=<rgba 0.75 0.6098 0.375 0.75>}
  NamedRerouteDeclaration(None=Max(A=NamedRerouteUsage(), B=NamedRerouteUsage())) {name=Mask_CombinedFrame, nodeColor=<rgba 0 0 0 1>}
  n74 = TextureCoordinate()
  n75 = Constant() {r=-0.096}
  n76 = Constant() {desc="When progress value changes, it defines where change ends.\r\n0 - 1", r=0.65}
  n77 = Subtract(A=ComponentMask(None=Add(A=@n74, B=@n75)) {r=true}, B=@n76) {desc=Offset by progress start}
  n78 = Constant() {r=2}
  n79 = Subtract(A=@n76, B=Constant() {desc="When progress value changes, it defines where change starts.\r\n0 - 1", r=0.16})
  n80 = OneMinus(None=Multiply(A=Multiply(A=@n77, B=@n78) {constB=4}, B=Divide(B=@n79)) {constB=4, desc=Shrink to progress zone})
  n81 = Divide(B=Constant() {r=1.2})
  n82 = Constant2Vector() {g=1.5, r=0.8}
  n83 = Max(
   A=Saturate(None=MaterialFunctionCall!(Value=@n80, Scale, Height=@n81) {materialFunction=/FortUILibrary/MaterialFunctions/Basics/MF_UI_PingPong}.Result)
   B=OneMinus(None=MaterialFunctionCall(Size=@n82) {materialFunction=&MF_UI_Box}.Fill)
  )
  n84 = Add(A=ComponentMask(None=@n74) {g=true}, B=LinearInterpolate(Alpha=Constant() {r=0.5}) {constA=0.08, constB=-0.08}) {constB=0}
  NamedRerouteDeclaration(None=MaterialFunctionCall(In=Add(A=@n83, B=@n84), Contrast=Constant() {r=50}) {materialFunction=&CheapContrast}.Result) {name=Notches, nodeColor=<rgba 0 0 0 1>}
  NamedRerouteUsage()
  OneMinus(None=@n44)
  n85 = Constant3Vector() {constant=<rgba 2 0.2297 0 1>}
  n86 = Constant3Vector() {constant=<rgba 0.2931 0.0915 2 1>}
  n87 = ScalarParameter() {parameterName=GradientBugBossColorSwitch}
  n88 = Constant3Vector() {constant=<rgba 3 0.5848 0 1>}
  n89 = NamedRerouteUsage()
  n90 = LinearInterpolate(B=LinearInterpolate(Alpha=ScalarParameter() {parameterName=Outer Bar Flash}) {constA=0.1}, Alpha=NamedRerouteUsage()) {constB=0.1}
  n91 = LinearInterpolate(A=Constant3Vector() {constant=<rgba 0.0182 0.0182 0.0182 1>}, Alpha=ScalarParameter() {parameterName=Progress Bar Flash})
  n92 = NamedRerouteUsage()
  n93 = NamedRerouteUsage()
  n94 = NamedRerouteUsage()
  n95 = ScalarParameter() {parameterName=BugBossProgressSwitch}
  n96 = Constant3Vector() {constant=<rgba 1 0.0658 0 1>}
  n97 = Constant3Vector() {constant=<rgba 0.0482 0.0212 0.1301 1>}
  n98 = ScalarParameter() {parameterName=BugBossProgressSwitch}
  n99 = Constant3Vector() {constant=<rgba 0.3813 0.1221 0.4397 1>}
  n100 = Constant3Vector() {constant=<rgba 1 0.3827 0 1>}
  n101 = LinearInterpolate(
   A=Multiply(A=Constant3Vector() {constant=<rgba 1 0.4327 0.1255 1>}, B=NamedRerouteUsage())
   B=MaterialFunctionCall(
    A=LinearInterpolate(A=@n96, B=@n97, Alpha=@n98)
    B=LinearInterpolate(A=@n46, B=@n99, Alpha=@n98)
    C=LinearInterpolate(A=@n100, B=@n47, Alpha=@n98)
    Alpha=MaterialFunctionCall() {materialFunction=&LinearGradient}.UGradient
   ) {materialFunction=&3ColorBlend}.Result
   Alpha=NamedRerouteUsage()
  ) {constAlpha=0}
  n102 = NamedRerouteUsage()
  n103 = Multiply(A=Constant3Vector() {constant=<rgba 2.5547 1.1568 2.0148 1>}, B=@n102)
  n104 = LinearInterpolate(
   A=LinearInterpolate(A=LinearInterpolate(A=@n90, B=@n91, Alpha=@n92), B=LinearInterpolate(A=@n93, B=@n94, Alpha=@n95), Alpha=NamedRerouteUsage())
   B=LinearInterpolate(A=@n101, B=LinearInterpolate(A=@n101, B=@n103, Alpha=@n102), Alpha=ScalarParameter() {parameterName=BugBossProgressSwitch})
   Alpha=Max(A=NamedRerouteUsage(), B=NamedRerouteUsage())
  )
  StaticSwitchParameter(
   True=LinearInterpolate(B=LinearInterpolate(A=@n85, B=@n86, Alpha=@n87), Alpha=NamedRerouteUsage())
   False=LinearInterpolate(A=@n104, B=LinearInterpolate(A=@n104, B=@n88, Alpha=@n89), Alpha=ScalarParameter() {parameterName=BomberBugKill?})
  ) {parameterName=Gradient}
  n105 = Multiply(A=NamedRerouteUsage()) {constB=0.2, desc=blend}
  n106 = Multiply(A=NamedRerouteUsage()) {constB=0.2, desc=blend}
  n107 = Add(A=Multiply(A=NamedRerouteUsage()) {constB=0.1}, B=NamedRerouteUsage())
  n108 = Multiply(A=@n107, B=NamedRerouteUsage())
  StaticSwitchParameter(
   True=Multiply(A=NamedRerouteUsage()) {desc=Blend}
   False=Max(A=Max(A=@n105, B=@n106), B=StaticSwitchParameter(True=@n108, False=@n107) {parameterName=Notched Progress})
  ) {parameterName=Gradient}
 ]
}
